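# Idle Niagara effect for the eighth Fortnite Birthday back bling, adding candle flames, glows, smoke and a light to skeletal mesh sockets.
# CandleFlames builds one partitioned ribbon per flame, bent against local owner velocity and scaled by System.FadeFlame.
# Six waveform-flickering glow sprites accompany six wavy smoke trails that fade and die with distance from their socket.
# A single flickering light follows the mesh, dimmed by time of day.
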NiagaraSystem /BRCosmetics/Accessories/FORT_Backpacks/Backpack_M_MED_FNBirthdayEight/FX/NS_BackPack_FNBirthday8_Idle_Update
dependencies {
 usedDataInterfaces [
  &NiagaraDataInterfaceCurve
  &NiagaraDataInterfaceParticleRead
  &NiagaraDataInterfaceVectorCurve
  &NiagaraDataInterfaceSkeletalMesh
  &NiagaraDataInterfaceArrayFloat3
 ]
 usedDynamicInputs []
 usedModules [
  &self:SystemSpawnScript.SystemScriptSource.SystemScriptGraph.NiagaraNodeAssignment_5.SetVariables_#1
  &self:SystemSpawnScript.SystemScriptSource.SystemScriptGraph.NiagaraNodeAssignment_7.SetVariables_#2
  &self:SystemSpawnScript.SystemScriptSource.SystemScriptGraph.NiagaraNodeAssignment_6.SetVariables_#3
  &self:SystemSpawnScript.SystemScriptSource.SystemScriptGraph.NiagaraNodeAssignment_8.SetVariables_#4
  &SystemState
 ]
}
emitters {
 CandleFlames {
  emitterSpawnScript [/SetVariables(Emitter.ParticlesPerFlame=7) {isSetParameters=true}]
  emitterUpdateScript [
   &EmitterState(
    Inactive Response=::NewEnumerator1
    Life Cycle Mode=::NewEnumerator1
    Loop Behavior=::NewEnumerator1
    Loop Duration Mode=::NewEnumerator1
    UseLoopDelay=false
   )
   &SpawnBurst_Instantaneous(
    Spawn Count=&Multiply_Int(A=$Emitter.ParticlesPerFlame, B=$System.FlameNumber)
    Use Spawn Probability=false
    UseLoopCountLimit=false
   )
  ]
  particleSpawnScript [
   &InitializeParticle(
    Lifetime=5
    Ribbon Width=20
    Ribbon Width Mode=::NewEnumerator1
    Sprite Size Mode=::NewEnumerator3
    Uniform Sprite Size=1
    UsePositionOffset=false
    Write Material Random=true
   )
   &PartitionParticles(
    Number Of Partitions=$System.FlameNumber
    Particles Per Partition:NiagaraInt32=$Emitter.ParticlesPerFlame
    Partition Mode:ENiagara_PartitionMode=::NewEnumerator1
   )
   /SetVariables(Particles.FlameVariation:NiagaraFloat=/Game/Effects/Niagara/DynamicInputs/Random/FN_SeededRandomRangeFloat(Minimum:NiagaraFloat=1, Maximum:NiagaraFloat=1, Random Seed:NiagaraInt32=$Particles.PartitionParticles.Partition)) {isSetParameters=true}
   /SetVariables(Particles.NormalizedPlaceInChain:NiagaraFloat=$Particles.PartitionParticles.Normalized Index In Partition) {isSetParameters=true}
   /SetVariables(Particles.RibbonWidth=&FloatFromCurve(
     FloatCurve=DataInterface() {propertyValues {Curve {keys [
         {arriveTangent=8.9306, interpMode=RCIM_Cubic, leaveTangent=8.9306, tangentMode=RCTM_User, time=0, value=0.3}
         {interpMode=RCIM_Cubic, tangentMode=RCTM_User, time=0.2, value=1}
         {arriveTangent=-1.0764, interpMode=RCIM_Cubic, leaveTangent=-1.0764, tangentMode=RCTM_User, time=1, value=0}
        ]}}}
     CurveIndex=$Particles.NormalizedPlaceInChain
     Scale Curve=&Lerp_Float(A=2.25, B=2.75, Alpha=$Particles.FlameVariation)
    )) {isSetParameters=true}
  ]
  particleUpdateScript [
   &ParticleState(Kill Particles When Lifetime Has Elapsed=false)
   &SkeletalMeshLocation(
    Bone / Socket Index=$Particles.PartitionParticles.Partition
    MaskSpawnGroup=false
    Mesh Sampling Type=::NewEnumerator5
    Position Sampling=::NewEnumerator2
    Sampled Position Offset=<0 0 -0.5>
    Skeletal Mesh=$System.Skeletal Mesh
    Socket Sampling Mode=::NewEnumerator5
   )
   /SetVariables(Particles.VelocityBender:Vector3f=&Multiply_VectorByFloat(
     Vector=/Niagara/DynamicInputs/Math/Clamp_Vector(
      Vector:Vector3f=&Multiply_VectorByFloat(
       Vector=&TransformVector(Vector=$Engine.Owner.Velocity, Source Space=::World, Destination Space=::Local)
       Float=-0.05
      )
      Min:Vector3f=&VectorFromFloat(Value=-2)
      Max:Vector3f=&VectorFromFloat(Value=2)
     )
     Float=&FloatFromCurve(
      FloatCurve=DataInterface() {propertyValues {Curve {keys [
          {arriveTangent=0.1033, interpMode=RCIM_Cubic, leaveTangent=0.1033, tangentMode=RCTM_User, time=0.2, value=0}
          {interpMode=RCIM_Cubic, tangentMode=RCTM_User, time=1, value=1}
         ]}}}
      CurveIndex=$Particles.NormalizedPlaceInChain
     )
    )) {isSetParameters=true}
   /SetVariables(Particles.RibbonU0Override=$Particles.NormalizedPlaceInChain) {isSetParameters=true}
   /SetVariables(Particles.GoalPosition:NiagaraPosition=&AddVectorToPosition(
     Position=$Particles.SkeletalMeshLocation.SampledPosition
     Vector=&Add_Vector(
      A=&Lerp_Vector(
       A=<0 0 0>
       B=&Multiply_VectorByFloat(Vector=&Lerp_Vector(A=<0 0 4>, B=<0 0 8>, Alpha=$Particles.FlameVariation), Float=$System.FadeFlame)
       Alpha=&FloatFromCurve(FloatCurve=DataInterface() {propertyValues {Curve {keys [<at 0 0>,<at 1 1>]}}}, CurveIndex=$Particles.NormalizedPlaceInChain)
      )
      B=$Particles.VelocityBender
     )
    )) {isSetParameters=true}
   &InterpolateOverTime(Rate Of Change=5, Target Vector Value:NiagaraPosition=$Particles.GoalPosition, Variable Type=::NewEnumerator2)
   /SetVariables(Particles.Position=&Lerp_Vector(
     A=Linked() {linkedVariable {name=Particles.GoalPosition, type=NiagaraPosition}}
     B=Linked() {linkedVariable {name=Particles.InterpolateOverTime.InterpolateOverTime_V3_.Moving Average, type=NiagaraPosition}}
     Alpha=&FloatFromCurve(
      FloatCurve=DataInterface() {propertyValues {Curve {keys [
          {arriveTangent=0.7977, interpMode=RCIM_Cubic, leaveTangent=0.7977, tangentMode=RCTM_User, time=0.1, value=0}
          {arriveTangent=3.1196, interpMode=RCIM_Cubic, leaveTangent=3.1196, tangentMode=RCTM_User, time=1, value=1}
         ]}}}
      CurveIndex=$Particles.NormalizedPlaceInChain
      Scale Curve=0
     )
    )) {isSetParameters=true}
   &SolveForcesAndVelocity(Clamp Velocity=false, Limit Acceleration=false)
   &DynamicMaterialParameters(
    Index 0 Param 1=&Add_Float(
     A=&Multiply_FloatByInt(Float=0.7, Integer=$Particles.PartitionParticles.Partition)
     B=&FN_SeededRandomRangeFloat(Minimum:NiagaraFloat=0, Maximum:NiagaraFloat=1, Random Seed:NiagaraInt32=$Particles.PartitionParticles.Partition)
    )
    Index 1 Param 3=&FloatFromCurve(FloatCurve=DataInterface() {propertyValues {Curve {keys [<at 0.5 0>,<at 1 1>]}}}, CurveIndex=$Particles.NormalizedPlaceInChain)
    Module.Index0Param0Write=true
    Module.Index0Param1Write=true
    Module.Index0Param2Write=true
    Module.Index0Param3Write=true
    Module.Index1Param0Write=true
    Module.Index1Param1Write=true
    Module.Index1Param2Write=true
    Module.Index1Param3Write=true
    Write Parameter Index 1=true
   )
   &ScaleColor(Scale Alpha=$System.FadeFlame, ScaleA=true, ScaleRGB=true)
   /SetVariables(Particles.RibbonWidth=&Multiply_Float(A=$Particles.Initial.RibbonWidth, B=$System.FadeFlame)) {isSetParameters=true}
  ]
  rendererClasses [&NiagaraRibbonRendererProperties]
  renderers [{rendererClass=&NiagaraRibbonRendererProperties}]
 }
 CandleGlows {
  emitterUpdateScript [
   &EmitterState(
    Inactive Response=::NewEnumerator1
    Life Cycle Mode=::NewEnumerator1
    Loop Behavior=::NewEnumerator1
    Loop Duration Mode=::NewEnumerator1
    UseLoopDelay=false
   )
   &SpawnBurst_Instantaneous(Spawn Count=6, Use Spawn Probability=false, UseLoopCountLimit=false)
  ]
  particleSpawnScript [
   &InitializeParticle(
    Color=$User.GlowColor
    Sprite Rotation Mode=::NewEnumerator1
    Sprite Size Mode=::NewEnumerator3
    Uniform Sprite Size=25
    UsePositionOffset=true
    Write Material Random=true
   )
   &PartitionParticles(Number Of Partitions=6)
  ]
  particleUpdateScript [
   &ParticleState(Kill Particles When Lifetime Has Elapsed=false)
   &self:StickMeshToRibbon_02(
    Index=3
    Particle Attribute Reader=DataInterface() {propertyValues {EmitterBinding {bindingMode=Other, emitterName=NE_Flame}}}
    ParticlePerRibbon=7
    Partition:NiagaraInt32=$Particles.PartitionParticles.Partition
    RibbonSpace=1
   ) {enabled=false, module=StickMeshToRibbon_02}
   &ScaleSpriteSize(
    Uniform Curve Index=$Output.StickMeshToRibbon_02.FlameVariation
    Uniform Curve Scale=&Multiply_Float(A=1, B=$System.FadeFlame)
    Uniform Curve Sprite Scale=DataInterface() {propertyValues {Curve {keys [{tangentMode=RCTM_User, time=0, value=0.5}, {tangentMode=RCTM_User, time=1, value=1}]}}}
   ) {enabled=false}
   &DynamicMaterialParameters(
    Index 0 Param 1=0.3
    Module.Index0Param0Write=true
    Module.Index0Param1Write=true
    Module.Index0Param2Write=true
    Module.Index0Param3Write=true
   )
   &SpriteFacingAndAlignment(Sprite Alignment=$Particles.RibbonDirectionVector, WriteAlignment=true, WriteFacing=false) {enabled=false}
   &SkeletalMeshLocation(
    Bone / Socket Index=$Particles.PartitionParticles.Partition
    MaskSpawnGroup=false
    Mesh Sampling Type=::NewEnumerator5
    Sampled Position Offset=<0 0 2>
    Skeletal Mesh=$System.Skeletal Mesh
    Socket Sampling Mode=::NewEnumerator5
   )
   &ScaleColor(
    Scale Alpha=&Waveform(
     Global Input Position=&RandomRangeFloat(Minimum=0, Maximum=1)
     "Waveform [1]"=::NewEnumerator0
     "[1] Frequency"=5
     "[1] Amplitude Min/Max"=<0.6 1>
    )
    ScaleA=true
    ScaleRGB=true
   )
  ]
  rendererClasses [&NiagaraSpriteRendererProperties]
  renderers [{rendererClass=&NiagaraSpriteRendererProperties}]
 }
 SmokeRibbon {
  emitterUpdateScript [
   &EmitterState()
   &FN_SpawnAnimTrail(Number of Anim Trails=6, Spawn Interval=2, Switch on ENiagara_AnimTrail_Source:ENiagara_AnimTrail_Source=::NewEnumerator1)
  ]
  particleSpawnScript [
   &FN_FlipRibbonV()
   &FN_InitializeAnimTrailBoneIndecies(Second Bone or Socket Index=0)
   &FN_InitializeAnimTrailBoneIndecies(AnimTrail Selection=1, First Bone or Socket Index=1) {module=FN_InitializeAnimTrailBoneIndecies005}
   &FN_InitializeAnimTrailBoneIndecies(AnimTrail Selection=2, First Bone or Socket Index=2, Second Bone or Socket Index=2) {module=FN_InitializeAnimTrailBoneIndecies004}
   &FN_InitializeAnimTrailBoneIndecies(AnimTrail Selection=3, First Bone or Socket Index=3, Second Bone or Socket Index=3) {module=FN_InitializeAnimTrailBoneIndecies003}
   &FN_InitializeAnimTrailBoneIndecies(AnimTrail Selection=4, First Bone or Socket Index=4, Second Bone or Socket Index=4) {module=FN_InitializeAnimTrailBoneIndecies002}
   &FN_InitializeAnimTrailBoneIndecies(AnimTrail Selection=5, First Bone or Socket Index=5, Second Bone or Socket Index=5) {module=FN_InitializeAnimTrailBoneIndecies001}
   &FN_InitializeAnimTrail(AnimTrailWidth=4, Color=<rgba 0.3802 0.3802 0.3802 0.35>, Lifetime=0.45, Position Offset=<0 0 2>, SkeletalMesh=$System.Skeletal Mesh)
   &AddVelocity(
    Coordinate Space=::World
    Velocity=&MakeVector(X=0, Y=0, Z=&Waveform("Waveform [1]"=::NewEnumerator9, "[1] Frequency"=3, "[1] Amplitude Min/Max"=<25 65>))
   )
   &AddVelocity(
    Rotation Coordinate Space=::World
    Velocity=&MakeVector(
     X=&Waveform(
      Global Input Position=&RandomRangeFloat(Minimum=0, Maximum=100, Randomness Mode=::NewEnumerator1, Random Seed=2, Fixed Random Seed=true)
      "Waveform [1]"=::NewEnumerator12
      "[1] Frequency"=4
      "[1] Amplitude Min/Max"=<-1 1>
      "[1] Scale"=&RandomRangeFloat(Minimum=15, Maximum=30)
     )
     Y=&Waveform(
      Global Input Position=&RandomRangeFloat(Minimum=0, Maximum=100, Randomness Mode=::NewEnumerator1, Random Seed=5, Fixed Random Seed=true)
      "Waveform [1]"=::NewEnumerator12
      "[1] Frequency"=6
      "[1] Amplitude Min/Max"=<-1 1>
      "[1] Scale"=&RandomRangeFloat(Minimum=15, Maximum=30)
     )
     Z=0
    )
   ) {module=AddVelocity001}
  ]
  particleUpdateScript [
   &ScaleVelocity(
    Coordinate Space:ENiagaraCoordinateSpace=::World
    Velocity Scale=&VectorFromCurve(VectorCurve=DataInterface() {propertyValues {
       XCurve {keys [
         <at 0 0 RCIM_Cubic>
         {arriveTangent=1.6667, interpMode=RCIM_Cubic, leaveTangent=1.6667, time=0.3, value=1}
         {arriveTangent=-1.8983, interpMode=RCIM_Cubic, leaveTangent=-1.8983, time=0.6, value=1}
         <at 0.7214 0.2 RCIM_Cubic>
         <at 0.8619 1 RCIM_Cubic>
        ]}
       YCurve {keys [
         <at 0 0 RCIM_Cubic>
         {arriveTangent=1.6667, interpMode=RCIM_Cubic, leaveTangent=1.6667, time=0.3, value=1}
         {arriveTangent=-1.8983, interpMode=RCIM_Cubic, leaveTangent=-1.8983, time=0.6, value=1}
         <at 0.7214 0.2 RCIM_Cubic>
         <at 0.8619 1 RCIM_Cubic>
        ]}
       ZCurve {keys [<at 0 1>,<at 1 1>]}
      }})
   )
   &SocketLocation(
    Position Sampling=::NewEnumerator1
    SamplePosOffset=true
    Socket Reader=$System.Socket Reader
    Socket Sampling Mode=::NewEnumerator3
    VelocitySpeedLimit=false
   )
   &SolveForcesAndVelocity(Clamp Velocity=false, Limit Acceleration=false)
   &FN_UpdateAnimTrail(SkeletalMesh=$System.Skeletal Mesh)
   &FN_RibbonMaterial()
   /SetVariables(Particles.RibbonDistance:NiagaraFloat=&DistanceBetweenPositions(A=$Particles.Position, B=$Output.SocketLocation.SampledPosition)) {isSetParameters=true}
   &ScaleRibbonWidth(Ribbon Width Scale=&FloatFromCurve(FloatCurve=DataInterface() {propertyValues {Curve {keys [{arriveTangent=4.5879, interpMode=RCIM_Cubic, time=0, value=0.3}, {interpMode=RCIM_Cubic, tangentMode=RCTM_User, time=1, value=1.6}]}}}))
   &ScaleColor(
    Scale Alpha=&FloatFromCurve(
     FloatCurve=DataInterface() {propertyValues {Curve {keys [{arriveTangent=-1.9717, interpMode=RCIM_Cubic, time=0, value=1}, <at 1 0 RCIM_Cubic>]}}}
     Scale Curve=&Multiply_Float(
      A=&FNTimeOfDayScalar(Night Scale=0.5, Morning Scale=1, Evening Scale=1)
      B=&FloatFromCurve(FloatCurve=DataInterface() {propertyValues {Curve {keys [<at 30 1 RCIM_Cubic>,<at 50 0 RCIM_Cubic>]}}}, CurveIndex=$Particles.RibbonDistance)
     )
    )
    Scale RGB=&MakeVectorFromLinearColorRGB(Color=&LerpLinearColors(
      StartColor=$User.GlowColor
      EndColor=<rgba 1 1 1 1>
      LerpFactor=&FloatFromCurve(FloatCurve=DataInterface() {propertyValues {Curve {
          defaultValue=1
          keys [{interpMode=RCIM_Cubic, tangentMode=RCTM_Break, time=0, value=0}, {arriveTangent=3.5776, tangentMode=RCTM_User, time=0.4953, value=0.9373}]
         }}})
     ))
    ScaleA=true
    ScaleRGB=true
   )
   &KillParticles(Kill Particles=&SetBoolByFloatComparison(A=$Particles.RibbonDistance, B=80))
  ]
  rendererClasses [&NiagaraRibbonRendererProperties]
  renderers [{rendererClass=&NiagaraRibbonRendererProperties}]
 }
 TopLight {
  emitterUpdateScript [
   &EmitterState()
   &SpawnBurst_Instantaneous(Use Spawn Probability=false, UseLoopCountLimit=false)
  ]
  particleSpawnScript [&InitializeParticle(
    Color=$User.LightColor
    Lifetime=$Emitter.CurrentLoopDuration
    Sprite Size Mode=::NewEnumerator3
    Uniform Sprite Size=50
    UsePositionOffset=false
    Write Material Random=true
   )]
  particleUpdateScript [
   &ParticleState()
   &SolveForcesAndVelocity(Clamp Velocity=false, Limit Acceleration=false)
   &SkeletalMeshLocation(
    Bone / Socket Index=5
    Bone Sampling Mode=::NewEnumerator4
    MaskSpawnGroup=false
    Sampled Position Offset=<10 0 0>
    Skeletal Mesh=$System.Skeletal Mesh
   )
   &ScaleColor(
    Scale Alpha=&Waveform(
     Waveform Count=::NewEnumerator1
     "Waveform [1]"=::NewEnumerator12
     "[1] Frequency"=1
     "[1] Amplitude Min/Max"=<0.1 1>
     "[2] Average"=false
     "Waveform [2]"=::NewEnumerator12
     "[2] Input Position"=0
     "[2] Phase"=$Emitter.Age
     "[2] Phase Scale"=1
     "[2] Compound Frequency"=2
     "[2] Frequency"=5
     "[2] Amplitude Min/Max"=<0.5 1.5>
     "[2] Scale"=1
     "[2] Bias"=0
     "[2] Exponent"=1
     "[2] Pulse Duration"=0.1
     "[2] Clip Triangle Min/Max"=<0 0>
     "[2] Random Seed"=0
    )
    ScaleA=true
    ScaleRGB=true
   )
   &FN_Light_Attributes(
    Color=$Particles.Color
    Create Color Attribute=true
    Create Exponent Attribute=true
    Create Radius Attribute=true
    Exponent=2
    Radius=75
    Scale Color=&FNTimeOfDayScalar(Night Scale=0.33, Morning Scale=0.5, Day Scale=0.5, Evening Scale=0.5)
   )
  ]
  rendererClasses [&NiagaraLightRendererProperties]
  renderers [{rendererClass=&NiagaraLightRendererProperties, rendererIndex=0}]
 }
}
summary {
 systemName=NS_BackPack_FNBirthday8_Idle_Update
 emitters [
  {emitterName=CandleFlames, rendererClasses [&NiagaraRibbonRendererProperties]}
  {emitterName=CandleGlows, rendererClasses [&NiagaraSpriteRendererProperties]}
  {emitterName=TopLight, rendererClasses [&NiagaraLightRendererProperties]}
  {emitterName=SmokeRibbon, rendererClasses [&NiagaraRibbonRendererProperties]}
 ]
 userVariables [
  User.bIsFrontEnd {type=NiagaraBool}
  User.bIsFrontEndPreview {type=NiagaraBool}
  User.FlameNumber {type=NiagaraInt32, defaultValue=100}
  User.GlowColor {type=LinearColor, description="", defaultValue {struct=LinearColor, value=<rgba 0.0479 2.8577 5 0.1>}}
  User.LightColor {type=LinearColor, description="", defaultValue {struct=LinearColor, value=<rgba 0.5473 2.705 7 1>}}
  User.MIFlames {
   type=MaterialInterface
   description=""
   defaultValue {struct=NiagaraExt_VariableValue_Object, value {object=null, objectClass=/Script/Engine.MaterialInterface}}
  }
  User.RibbonIDOffset {type=NiagaraInt32}
 ]
}
